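; Placeholder cosmetic glider Blueprint that applies a faint hologram look to the parachute mesh during gameplay.
; When chute visibility changes it calls the parent, then, outside frontend display contexts and only when the chute becomes visible, calls SetHologramIntensity.
; That custom event sets the mesh materials' hologram color blend to the passed intensity and zeroes the primary wave intensity.
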(event EventOnChuteVisibilityChanged
  (params (bParachuteVisibility :type Boolean))
  (|CallParentFunction bParachuteVisibility)
  (if (Reactive|Frontend|IsCosmeticObjectDisplayedinFrontendContext 0)
    _
    (elif bParachuteVisibility
      (CallFunction|SetHologramIntensity 0.25))))

(event Custom|SetHologramIntensity
  (params (NewParam :type Float%20%28double-precision%29))
  (bind _parachutemesh (Variables|Parachute|GetParachuteMesh))
  (Rendering|Material|SetScalarParameterValueonMaterials _parachutemesh "HologramColorBlend" NewParam)
  (Rendering|Material|SetScalarParameterValueonMaterials _parachutemesh "PrimaryWaveIntensity" 0.0))

(fn ConstructionScript
  (|CallParentFunction))
