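; B_PartModifier_Cosmetic_Placeholder is a cosmetic part modifier that applies a partial hologram look to character parts outside the frontend.
; When character customization finishes, it skips frontend display contexts, then caches the pawn's head, body, face and backpack skeletal meshes.
; A SetHologramIntensity event loops over those meshes, setting the HologramColorBlend material parameter to the given intensity and zeroing PrimaryWaveIntensity.
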(event EventOnCharacterCustomizationFinished
  (params (CharacterPartInfo :type Array%20of%20Applied%20Character%20Part%20Info%20Structures))
  (|CallParentFunction CharacterPartInfo)
  (if (not (or (Reactive|Frontend|IsCosmeticObjectDisplayedinFrontendContext 0) (CustomCosmeticModifier|IsinFrontEnd)))
    (Variables|Default|SetMeshes (Utilities|Array|MakeArray (Customization|GetSkeletalMeshforPartType (CustomCharacter|GetAssociatedPlayerPawn) "Head" :!class /Script/FortniteGame.FortPlayerPawn) (Customization|GetSkeletalMeshforPartType (CustomCharacter|GetAssociatedPlayerPawn) "Body" :!class /Script/FortniteGame.FortPlayerPawn) (Customization|GetSkeletalMeshforPartType (CustomCharacter|GetAssociatedPlayerPawn) "Face" :!class /Script/FortniteGame.FortPlayerPawn) (Customization|GetSkeletalMeshforPartType (CustomCharacter|GetAssociatedPlayerPawn) "Backpack" :!class /Script/FortniteGame.FortPlayerPawn)))
    (CallFunction|SetHologramIntensity 0.25)))

(event Custom|SetHologramIntensity
  (params (NewParam :type Float%20%28double-precision%29))
  (for _array_element (Variables|Default|GetMeshes)
    (Rendering|Material|SetScalarParameterValueonMaterials _array_element "HologramColorBlend" NewParam)
    (Rendering|Material|SetScalarParameterValueonMaterials _array_element "PrimaryWaveIntensity" 0.0)))

(fn ConstructionScript
  (|CallParentFunction))
