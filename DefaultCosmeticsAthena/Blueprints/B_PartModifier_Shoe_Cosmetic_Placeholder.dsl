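; Part modifier for a placeholder shoe cosmetic that adjusts its hologram material effect outside the frontend.
; When character customization finishes, it calls the parent handler, then skips the adjustment if the cosmetic is displayed in a frontend context.
; Otherwise it sets a partial hologram color blend and zeroes the primary wave intensity through scalar material parameters on the owner's part skeletal mesh.
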(event EventOnCharacterCustomizationFinished
  (params (CharacterPartInfo :type Array%20of%20Applied%20Character%20Part%20Info%20Structures))
  (|CallParentFunction CharacterPartInfo)
  (if (not (or (Reactive|Frontend|IsCosmeticObjectDisplayedinFrontendContext 0) (CustomCosmeticModifier|IsinFrontEnd)))
    (CallFunction|SetHologramIntensity 0.25)))

(event Custom|SetHologramIntensity
  (params (NewParam :type Float%20%28double-precision%29))
  (bind _returnvalue (CustomCharacter|GetSkeletalMeshforPartonOwner))
  (Rendering|Material|SetScalarParameterValueonMaterials _returnvalue "HologramColorBlend" NewParam)
  (Rendering|Material|SetScalarParameterValueonMaterials _returnvalue "PrimaryWaveIntensity" 0.0))

(fn ConstructionScript
  (|CallParentFunction))
