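; Placeholder pickaxe cosmetic Blueprint that gives the weapon mesh a hologram tint when shown outside the frontend.
; On weapon cosmetics init, it calls the parent, then applies a partial hologram intensity unless the object is displayed in a frontend context.
; The SetHologramIntensity event drives the HologramColorBlend material parameter on the weapon mesh and always zeroes PrimaryWaveIntensity.
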(event EventOnInitWeaponCosmetics
  (|CallParentFunction)
  (if (not (Reactive|Frontend|IsCosmeticObjectDisplayedinFrontendContext 0))
    (CallFunction|SetHologramIntensity 0.25)))

(event Custom|SetHologramIntensity
  (params (NewParam :type Float%20%28double-precision%29))
  (bind _weaponmesh (Variables|Visual|GetWeaponMesh))
  (Rendering|Material|SetScalarParameterValueonMaterials _weaponmesh "HologramColorBlend" NewParam)
  (Rendering|Material|SetScalarParameterValueonMaterials _weaponmesh "PrimaryWaveIntensity" 0.0))

(fn ConstructionScript
  (|CallParentFunction))
